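# Material Gym example that exercises the MF_CameraAlignedVirtualPlane material function, which by its name generates UVs on a camera-aligned virtual plane.
# Scalar parameters for X and Y scale form the function's UV tiling, and a vector parameter supplies its offset.
# The resulting UVs sample the UV Editor color grid texture.
# The material is Default Lit and flagged for skeletal mesh and mesh deformer use, with automatic usage setting disabled.
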Material /Game/Global/Materials/MaterialGym/Materials/MaterialElements/UVs/CameraAlignedVirutalPlane/M_CameraAlignedVirtualPlane
graph {
 properties {bAutomaticallySetUsageInEditor=false, bUsedWithMeshDeformer=true, bUsedWithSkeletalMesh=true, usedShadingModels=Default Lit}
 nodes [
  n1 = MaterialFunctionCall!(
   UV Tiling=AppendVector(
    A=ScalarParameter() {defaultValue=2, group=CameraAlignedVirtualPlane, parameterName=UVScaleX, sortPriority=0}
    B=ScalarParameter() {defaultValue=2, group=CameraAlignedVirtualPlane, parameterName=UVScaleY, sortPriority=1}
   )
   UV Offset=VectorParameter() {parameterName=UVOffset, sortPriority=2, defaultValue=<rgba 0.228 0.256 0 1>}.RGB
  ) {materialFunction=&/MF_CameraAlignedVirtualPlane}
  TextureSample(UVs=@n1.OutputUV) {texture=/UVEditor/Textures/UVEditorColorGrid_Color}
 ]
}
