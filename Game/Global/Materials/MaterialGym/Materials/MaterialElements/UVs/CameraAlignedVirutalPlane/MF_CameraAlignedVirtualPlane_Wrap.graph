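# Material function that wraps MF_CameraAlignedVirtualPlane to supply camera-aligned virtual plane UVs with material-tunable settings.
# It exposes X and Y UV tiling and offset as scalar parameters in the MF_CVP group, appends each pair into a vector and passes them to the inner function.
# The resulting UVs are returned through a single UV output.
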MaterialFunction /Game/Global/Materials/MaterialGym/Materials/MaterialElements/UVs/CameraAlignedVirutalPlane/MF_CameraAlignedVirtualPlane_Wrap
graph {nodes [
  n1 = MaterialFunctionCall!(
   UV Tiling=AppendVector(
    A=ScalarParameter() {defaultValue=2, group=MF_CVP, parameterName=CVP_UVTilingX}
    B=ScalarParameter() {defaultValue=2, group=MF_CVP, parameterName=CVP_UVTilingY}
   )
   UV Offset=AppendVector(
    A=ScalarParameter() {group=MF_CVP, parameterName=CVP_UVOffsetX}
    B=ScalarParameter() {defaultValue=0.5, group=MF_CVP, parameterName=CVP_UVOffsetY}
   )
  ) {materialFunction=&/MF_CameraAlignedVirtualPlane}
  FunctionOutput(None=@n1.OutputUV) {description=camera aligned virtual plane uvs, outputName=UV}
 ]}
