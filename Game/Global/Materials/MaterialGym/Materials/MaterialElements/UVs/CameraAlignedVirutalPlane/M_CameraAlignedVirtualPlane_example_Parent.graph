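# Example parent material in the MaterialGym library that demonstrates UVs projected on a camera-aligned virtual plane.
# The virtual plane function supplies UVs to a polar UV remap, whose output drives the input UVs of an animated storm noise function.
# All three stages are wrapper material functions, so the graph itself holds almost no logic.
# It is Default Lit and flagged for skeletal mesh and mesh deformer use.
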Material /Game/Global/Materials/MaterialGym/Materials/MaterialElements/UVs/CameraAlignedVirutalPlane/M_CameraAlignedVirtualPlane_example_Parent
graph {
 properties {bAutomaticallySetUsageInEditor=false, bUsedWithMeshDeformer=true, bUsedWithSkeletalMesh=true, usedShadingModels=Default Lit}
 nodes [
  Constant()
  n1 = MaterialFunctionCall() {materialFunction=&/MF_CameraAlignedVirtualPlane_Wrap}
  MaterialFunctionCall!(InputUV=MaterialFunctionCall!(UVInput=@n1.UV, PolarUV_CenterOffsetR, PolarUV_CenterOffsetG) {materialFunction=/Game/Global/Materials/MaterialGym/Materials/MaterialElements/UVs/UVMainpulation/PolarUV/MF_Polar_Wrap}.OutputUV) {materialFunction=/Game/Global/Materials/MaterialGym/Materials/MaterialElements/UVs/AnimatedNoise/MF_StormNoise_Wrap}
 ]
}
